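# Animation Blueprint for the Citadel backpack that opens and closes the pack as the player equips or unequips the Citadel pickaxe.
# Initialization caches the owning FortPlayerPawn; each update sets a flag by substring-matching the current weapon's object name.
# The PlayShieldAnim state machine uses that flag to move from Idle's ref pose to the one-shot open, then close animation.
# Close returns to Idle automatically when it finishes, and all transitions have zero crossfade.
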AnimBlueprint /BRCosmetics/Accessories/FORT_Backpacks/Backpack_M_MED_Citadel/Meshes/M_MED_Citadel_Pack_AnimBP
target_skeleton=&/M_MED_Citadel_Pack_Skeleton
graphs [
 AnimGraph {nodes [
   Input Pose()
   AnimGraphNode_StateMachine_0 = PlayShieldAnim() {is=title AnimGraphNode_StateMachine}
   Output Pose(Result=@AnimGraphNode_StateMachine_0) {layerGroup=None}
  ]}
 PlayShieldAnim {
  entry=Idle
  path=AnimGraph.AnimGraphNode_StateMachine_0.PlayShieldAnim
  ids=id
  nodes [
   AnimStateNode_0 = Idle()
   AnimStateNode_1 = Equipped() {is=title AnimStateNode}
   AnimStateNode_2 = Unequipped() {is=title AnimStateNode}
   AnimStateTransitionNode_0 = AnimStateTransitionNode() {crossfadeDuration=0}
   AnimStateTransitionNode_1 = AnimStateTransitionNode() {crossfadeDuration=0}
   AnimStateTransitionNode_2 = AnimStateTransitionNode() {bAutomaticRuleBasedOnSequencePlayerInState=true, crossfadeDuration=0}
  ]
  links [
   {
    dst=Unequipped
    rule=AnimGraph.AnimGraphNode_StateMachine_0.PlayShieldAnim.AnimStateTransitionNode_1.Transition
    src=Equipped
    via=AnimStateTransitionNode_1
   }
   {
    dst=Equipped
    rule=AnimGraph.AnimGraphNode_StateMachine_0.PlayShieldAnim.AnimStateTransitionNode_0.Transition
    src=Idle
    via=AnimStateTransitionNode_0
   }
   {
    dst=Idle
    rule=AnimGraph.AnimGraphNode_StateMachine_0.PlayShieldAnim.AnimStateTransitionNode_2.Transition
    src=Unequipped
    via=AnimStateTransitionNode_2
   }
  ]
 }
 Idle {
  path=AnimGraph.AnimGraphNode_StateMachine_0.PlayShieldAnim.AnimStateNode_0.Idle
  nodes [Output Animation Pose(Result=Local Space Ref Pose())]
 }
 Equipped {
  path=AnimGraph.AnimGraphNode_StateMachine_0.PlayShieldAnim.AnimStateNode_1.Equipped
  nodes [Output Animation Pose(Result=Sequence Player 'Citadel_Open_Pack'!(bLoopAnimation:bool=False) -> (Pose) {bLoopAnimation=false, sequence=/BRCosmetics/Accessories/FORT_Backpacks/Backpack_M_MED_Citadel/Animation/Citadel_Open_Pack})]
 }
 Unequipped {
  path=AnimGraph.AnimGraphNode_StateMachine_0.PlayShieldAnim.AnimStateNode_2.Unequipped
  nodes [Output Animation Pose(Result=Sequence Player 'Citadel_Close_Pack'!(bLoopAnimation:bool=False) -> (Pose) {bLoopAnimation=false, sequence=/BRCosmetics/Accessories/FORT_Backpacks/Backpack_M_MED_Citadel/Animation/Citadel_Close_Pack})]
 }
 Transition {
  path=AnimGraph.AnimGraphNode_StateMachine_0.PlayShieldAnim.AnimStateTransitionNode_0.Transition
  nodes [Result(bCanEnterTransition=Get IsCitadelPickaxeEquipped() -> (IsCitadelPickaxeEquipped:bool))]
 }
 Transition {
  path=AnimGraph.AnimGraphNode_StateMachine_0.PlayShieldAnim.AnimStateTransitionNode_1.Transition
  nodes [Result(bCanEnterTransition=NOT Boolean(A=Get IsCitadelPickaxeEquipped()))]
 }
 Transition {path=AnimGraph.AnimGraphNode_StateMachine_0.PlayShieldAnim.AnimStateTransitionNode_2.Transition, nodes [Result()]}
 EventGraph {nodes [
   Event BlueprintUpdateAnimation()
   n1 = GetCurrentWeapon!(self:Fort Pawn Object Reference=Get As Fort Player Pawn() -> (As Fort Player Pawn:Fort Player Pawn Object Reference)) -> (ReturnValue:Fort Weapon Object Reference)
   > Set IsCitadelPickaxeEquipped(IsCitadelPickaxeEquipped:bool=Contains(SearchIn="To String (Object)"!(InObj:Object Reference=@n1) -> (ReturnValue:String), Substring=B_Athena_Pickaxe_Citadel)) -> (Output_Get:bool)
   Event BlueprintInitializeAnimation()
   > n2 = Cast To FortPlayerPawn(Object=TryGetPawnOwner())
   > Set As Fort Player Pawn(As Fort Player Pawn:Fort Player Pawn Object Reference=@n2) -> (Output_Get:Fort Player Pawn Object Reference)
  ]}
]
variables [
 As Fort Player Pawn {type=FortPlayerPawn}
 IsCitadelPickaxeEquipped {type=boolean}
]
